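# Translucent, two-sided unlit material for Niagara mesh particles that renders flaming wing effects in the CoatCheck player emote.
# Panning normal-map distortion warps fire noise, which feeds a tinted edge/core/highlight gradient over a background glow, then hue shift and emissive adjustment.
# Distorted wing-shape and logo textures supply named masks.
# Dynamic-parameter erosion, vertex and particle color, and quality-switched depth fade build opacity; a normal-directed wobble is also computed.
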Material /Game/Animation/Game/MainPlayer/Emotes/CoatCheck/FX/M_FX_Emote_CoatCheck_Wing
graph {
 properties {
  bUsedWithNiagaraMeshParticles=true
  blendMode=BLEND_Translucent
  pixelDepthOffsetMode=PDOM_Legacy
  shadingModel=MSM_Unlit
  twoSided=true
  usedShadingModels=Unlit
 }
 nodes [
  Constant3Vector()
  n1 = VectorParameter() {parameterName=BackgroundGlowColor, defaultValue=<rgba 0 0 0 1>}
  n2 = VectorParameter() {group=Color, parameterName=CoreColor, defaultValue=<rgba 0 0 0 1>}
  n3 = VectorParameter() {group=Color, parameterName=HighlightColor, defaultValue=<rgba 0 0 0 1>}
  n4 = NamedRerouteUsage()
  n5 = Subtract(A=NamedRerouteUsage(), B=Multiply(A=NamedRerouteUsage(), B=ScalarParameter() {group=Color, parameterName=NoiseColorOffsets}) {constB=0.35})
  n6 = Multiply(
   A=ParticleColor().RGB
   B=LinearInterpolate(
    A=VectorParameter() {group=Color, parameterName=EdgeColor, defaultValue=<rgba 0 0 0 1>}.RGB
    B=LinearInterpolate(A=@n2.RGB, B=@n3.RGB, Alpha=@n4)
    Alpha=Saturate(None=@n5)
   )
  )
  n7 = ParticleRandom()
  n8 = TextureCoordinate()
  n9 = Abs(None=@n8)
  n10 = VectorParameter() {parameterName=UV_Scale, defaultValue=<rgba 1 1.5 0 1>}
  n11 = Add(
   A=@n7
   B=Multiply(A=NamedRerouteDeclaration(None=@n9) {name=BaseUV, nodeColor=<rgba 0.875 0 1 1>}, B=ComponentMask(None=@n10.RGB) {g=true, r=true})
  )
  n12 = ComponentMask(None=VectorParameter() {group=Distortion, parameterName=Distortion UV Scale, sortPriority=3, defaultValue=<rgba 1 1 0 0>}.RGB) {g=true, r=true}
  n13 = Time()
  n14 = ScalarParameter() {defaultValue=1, group=Panner Speed, parameterName=Speed}
  n15 = TextureSampleParameter2D(UVs=Panner(Coordinate=Multiply(A=@n11, B=@n12), Time=Multiply(A=@n13, B=@n14)) {speedY=1}) {
   group=Distortion
   parameterName=Distortion Normal
   samplerType=SAMPLERTYPE_Normal
   sortPriority=2
   texture=&T_Weap_Distort_N_02
  }
  n16 = ComponentMask(None=@n15.RGB) {g=true, r=true}
  n17 = VectorParameter() {group=Fire Shape, parameterName=Fire_MacroShape_UVScale, sortPriority=5, defaultValue=<rgba 1 1 0 0>}
  n18 = ComponentMask(None=Multiply(A=NamedRerouteUsage(), B=ComponentMask(None=@n17.RGB) {g=true, r=true})) {g=true}
  n19 = LinearInterpolate(
   B=Divide(
    A=LinearInterpolate(A=@n16, Alpha=@n18) {constB=0}
    B=ScalarParameter() {defaultValue=3, group=Distortion, parameterName=UV_Distortion_Divsor, sortPriority=1}
   ) {constB=20}
   Alpha=ScalarParameter() {defaultValue=1, group=Distortion, parameterName=UV_Distortion_Lerp}
  )
  n20 = Add(A=@n7, B=Multiply(A=NamedRerouteUsage(), B=Constant2Vector() {g=0.9, r=0.8}) {constB=0.8})
  n21 = ComponentMask(None=VectorParameter() {group=Fire Shape, parameterName=Fire_Micro_UVScale, sortPriority=8, defaultValue=<rgba 1 1 0 0>}.RGB) {g=true, r=true}
  n22 = Panner(
   Coordinate=MaterialFunctionCall(XY=Multiply(A=@n20, B=@n21)) {materialFunction=&Swizzle}.YX
   Time=Multiply(A=Time(), B=ScalarParameter() {defaultValue=1, group=Panner Speed, parameterName=Speed})
  ) {speedX=1.2, speedY=0.2}
  n23 = TextureSampleParameter2D(UVs=ShadingPathSwitch(Default=Add(A=@n19, B=@n22), Mobile=@n22) {desc=Migrated from FeatureLevelSwitch}) {
   group=Fire Shape
   parameterName=Fire Micro Texture
   samplerType=SAMPLERTYPE_LinearColor
   sortPriority=6
   texture=&T_NoisePack_LinearColor
  }
  n24 = TextureSample() {samplerType=SAMPLERTYPE_LinearGrayscale, texture=/Engine/Functions/Engine_MaterialFunctions02/ExampleContent/Textures/LowResBlurredNoise}
  n25 = Saturate(None=LinearInterpolate(
    A=ChannelMaskParameter(None=@n23.RGBA) {group=Fire Shape, parameterName=Fire Micro Mask, sortPriority=7}.R
    B=@n24.R
    Alpha=ScalarParameter() {defaultValue=0.2, parameterName=Fire Micro lerp}
   ) {constAlpha=0.2})
  n26 = NamedRerouteUsage()
  n27 = ScalarParameter() {defaultValue=6, group=NoiseControl, parameterName=VertexColorFade_Exponent, sortPriority=3}
  n28 = Multiply(A=@n25, B=Multiply(A=Power(Base=@n26, Exp=@n27) {constExponent=4}, B=Constant() {r=1}))
  n29 = LinearInterpolate(
   A=ScalarParameter() {defaultValue=3, parameterName=EdgeExponent}
   B=ScalarParameter() {defaultValue=0.5, group=NoiseControl, parameterName=CoreExponent}
   Alpha=NamedRerouteUsage()
  ) {constAlpha=0}
  n30 = NamedRerouteDeclaration(None=Multiply(A=Power(Base=@n28, Exp=@n29))) {name=MainNoise, nodeColor=<rgba 1 0 0.75 1>}
  n31 = ScalarParameter() {defaultValue=1, group=NoiseControl, parameterName=OverallBrightness, sortPriority=1}
  n32 = ScalarParameter() {defaultValue=0.5, group=NoiseControl, parameterName=Core_Pos_Min}
  n33 = ScalarParameter() {defaultValue=0.6, group=NoiseControl, parameterName=Core_Pos_Max}
  n34 = NamedRerouteUsage()
  n35 = Saturate(None=Add(A=Multiply(A=@n30, B=@n31) {constB=3}, B=SmoothStep(Min=@n32, Max=@n33, Value=@n34)))
  n36 = Saturate(None=Add(A=@n35, B=OneMinus(None=@n1.A)))
  MaterialFunctionCall(Emissive=MaterialFunctionCall(Hue Shift Percentage=Constant() {r=-0.01}, Texture=LinearInterpolate(A=@n1.RGB, B=@n6, Alpha=@n36)) {materialFunction=&HueShift}.Result) {materialFunction=&MF_FX_EmissiveAdjustments}
  n37 = VertexNormalWS()
  n38 = TextureSample(UVs=Panner(Coordinate=TextureCoordinate()) {speedY=-1}) {samplerType=SAMPLERTYPE_Normal, texture=&T_Weap_Distort_N_02}
  Multiply(A=Multiply(A=Multiply(A=@n37, B=@n38.R), B=ScalarParameter() {group=VS, parameterName=VS_WobbleStr}), B=OneMinus(None=VertexColor().G))
  NamedRerouteDeclaration(None=@n25) {name=NoiseField, nodeColor=<rgba 1 0 0.0937 1>}
  n39 = NamedRerouteUsage()
  n40 = Multiply(
   A=Multiply(A=NamedRerouteUsage(), B=ScalarParameter() {group=MacroShape, parameterName=DistortionEffect})
   B=OneMinus(None=TextureSample(UVs=@n39) {samplerType=SAMPLERTYPE_Masks, texture=&/T_FX_CoatCheck_WingShape}.R)
  )
  NamedRerouteDeclaration(None=TextureSample(UVs=Add(A=@n39, B=@n40)) {samplerType=SAMPLERTYPE_Masks, texture=&T_FX_CoatCheck_WingShape}.R) {name=FlamesMask, nodeColor=<rgba 0 1 0.6484 1>}
  n41 = Subtract(A=Multiply(A=NamedRerouteUsage()) {constB=2})
  n42 = VectorParameter() {group=Logo, parameterName=Logo_NoiseDistort, defaultValue=<rgba 0 0 0 1>}
  n43 = Multiply(A=ComponentMask(None=@n42.RGB) {g=true, r=true}, B=@n42.B)
  n44 = TextureSample(UVs=Add(A=TextureCoordinate() {coordinateIndex=2}, B=Multiply(A=@n41, B=@n43))) {samplerType=SAMPLERTYPE_Masks, texture=&/T_FX_CoatCheck_Icon}
  NamedRerouteDeclaration(None=@n44.G) {name=LogoMask, nodeColor=<rgba 0 0.9766 1 1>}
  NamedRerouteDeclaration(None=@n44.R) {name=LogoGlow, nodeColor=<rgba 0.0078 0 1 1>}
  NamedRerouteDeclaration(None=@n16) {name=DistortionNrms, nodeColor=<rgba 0.2656 0 1 1>}
  n45 = Subtract(A=Multiply(A=NamedRerouteUsage()) {constB=2})
  n46 = VectorParameter() {parameterName=Mask_UV_Distortion, defaultValue=<rgba 0 0 0 1>}
  n47 = Multiply(A=AppendVector(A=@n46.R, B=@n46.R), B=@n46.B)
  NamedRerouteDeclaration(None=Add(A=TextureCoordinate() {coordinateIndex=1}, B=Multiply(A=@n45, B=@n47))) {name=DistortedBaseUV, nodeColor=<rgba 0.7813 0 1 1>}
  n48 = DynamicParameter() {defaultValue=<rgba 0 0 0 0>, paramNames [Erode, ErosionDirectionBlend, Param3, Param4]}
  n49 = ScalarParameter() {parameterName=erode_width}
  n50 = OneMinus(None=VertexColor().G)
  n51 = NamedRerouteUsage()
  n52 = NamedRerouteUsage()
  n53 = Add(A=@n35, B=Multiply(A=NamedRerouteUsage(), B=@n1.A))
  n54 = Multiply(
   A=SmoothStep(Min=@n48.Erode, Max=Add(A=@n48.Erode, B=@n49), Value=Add(A=@n50, B=@n51))
   B=Multiply(A=Saturate(None=@n53), B=OneMinus(None=@n52))
  )
  n55 = Multiply(A=OneMinus(None=VertexColor().R), B=Multiply(A=Saturate(None=@n54), B=ParticleColor().A))
  n56 = DepthFade(Opacity=@n55) {fadeDistanceDefault=50}
  QualitySwitch(Default=@n55, High=@n56, Epic=@n56)
  Subtract(A=@n26, B=NamedRerouteUsage())
  VectorParameter() {group=Color, parameterName=OverallTint, defaultValue=<rgba 1 1 1 1>}
  n57 = ComponentMask(None=VectorParameter() {parameterName=RadialUVOffset, defaultValue=<rgba 0 0 0 1>}.RGB) {g=true, r=true}
  VertexInterpolator(VS=MaterialFunctionCall(Vector or UVs=Subtract(A=@n8, B=@n57) {constB=0.5}, Swizzle Coordinate Output=StaticBoolParameter() {parameterName=RadialSwizzle}) {materialFunction=&VectorToRadialValue}.Radial Coordinates)
 ]
}
