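# Additive, unlit, two-sided Niagara particle material for the wing glows in the CoatCheck emote.
# Color blends radially from a CoreColor parameter to particle color, then passes through a shared emissive-adjustment function.
# Opacity comes from cloud noise in radial coordinates, panned by particle age with a random offset, shaped by a sphere mask, vertical fade and particle alpha.
# A quality switch skips depth fade on Low and Medium.
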Material /Game/Animation/Game/MainPlayer/Emotes/CoatCheck/FX/M_FX_Emote_CoatCheck_WingGlows
graph {
 properties {
  bUsedWithNiagaraMeshParticles=true
  bUsedWithNiagaraSprites=true
  blendMode=BLEND_Additive
  pixelDepthOffsetMode=PDOM_Legacy
  shadingModel=MSM_Unlit
  twoSided=true
  usedShadingModels=Unlit
 }
 nodes [
  n1 = VectorParameter() {parameterName=CoreColor, defaultValue=<rgba 0 1 0 1>}
  n2 = ParticleColor()
  n3 = MaterialFunctionCall(Vector or UVs=Subtract(A=TextureCoordinate(), B=Constant2Vector() {r=0.5})) {materialFunction=&VectorToRadialValue}
  n4 = Multiply(A=@n3.Linear Distance, B=@n1.A) {constB=1.5}
  MaterialFunctionCall(Emissive=LinearInterpolate(A=@n1.RGB, B=@n2.RGB, Alpha=Saturate(None=@n4))) {materialFunction=&MF_FX_EmissiveAdjustments}
  n5 = Power(Base=SphereMask(A=TextureCoordinate() {vTiling=0.5}, B=Constant2Vector() {r=0.5}) {attenuationRadius=0.5, hardnessPercent=0})
  Multiply(A=@n5, B=VertexColor().R)
  n6 = ComponentMask(None=TextureCoordinate()) {g=true}
  OneMinus(None=@n6)
  n7 = Constant2Vector() {g=0.1, r=2}
  n8 = ParticleRelativeTime()
  n9 = ParticleRandom()
  n10 = TextureSample(UVs=Panner(Coordinate=Multiply(A=@n3.Radial Coordinates, B=@n7), Time=Add(A=@n8, B=@n9)) {speedY=-0.15}) {texture=/Game/Effects/Fort_Effects/Textures/Smoke/T_Clouds_01}
  n11 = Multiply(A=@n6) {constB=5}
  n12 = Multiply(A=Multiply(A=Power(Base=@n10.G) {constExponent=3}, B=@n2.A), B=Multiply(A=@n5, B=Saturate(None=@n11)))
  QualitySwitch(Default=DepthFade(Opacity=@n12), Low=@n12, Medium=@n12)
 ]
}
